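# Physics asset for the NoiseClue guitar cosmetic in Sparks, giving its hanging chain bodies for physics simulation.
# Capsules cover the handle, the chain base and five dynamic chain links.
# Constraints run in series from the handle through the base to the last link.
# The base-to-handle joint rotates freely, while the link joints carry swing limits and a very tight twist limit.
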PhysicsAsset /SparksCosmetics/Instrument/Guitar/NoiseClue/Meshes/NoiseClue_Guitar_Physics
bodies {
 base_chain [{shapeType=Capsule, length=0.505, radius=0.505, shapeName=base_chain_capsule, rotation=<rot 0 -90 0>}]
 dyn_chain01 [{shapeType=Capsule, length=2, radius=0.5, shapeName=dyn_chain01_capsule, rotation=<rot 90 0 0>}]
 dyn_chain02 [{shapeType=Capsule, length=2, radius=0.5, shapeName=dyn_chain02_capsule, center=<0 0 -1.95>, rotation=<rot 90 0 0>}]
 dyn_chain03 [{shapeType=Capsule, length=2, radius=0.5, shapeName=dyn_chain03_capsule, center=<0 0 -2.45>, rotation=<rot 90 0 0>}]
 dyn_chain04 [{shapeType=Capsule, length=2, radius=0.5, shapeName=dyn_chain04_capsule, center=<0 0 -1.9>, rotation=<rot 90 0 0>}]
 dyn_chain05 [{shapeType=Capsule, length=2, radius=0.5, shapeName=dyn_chain05_capsule, center=<0 0 -1.6>, rotation=<rot 90 0 0>}]
 handle [{shapeType=Capsule, length=10, radius=4, shapeName=handle_capsule, center=<-1.0567 -0.5684 66.7851>, rotation=<rot 6.8944 -91.2126 -0.1419>}]
}
constraints [
 {
  bone1Name=base_chain
  bone2Name=handle
  swing1LimitDegrees=45
  swing1Motion=Free
  swing2LimitDegrees=45
  swing2Motion=Free
  twistLimitDegrees=45
  twistMotion=Free
 }
 {bone1Name=dyn_chain01, bone2Name=base_chain, swing1LimitDegrees=45, swing2LimitDegrees=45, twistLimitDegrees=1}
 {bone1Name=dyn_chain02, bone2Name=dyn_chain01, swing1LimitDegrees=45, swing2LimitDegrees=45, twistLimitDegrees=1}
 {bone1Name=dyn_chain03, bone2Name=dyn_chain02, swing1LimitDegrees=45, swing2LimitDegrees=45, twistLimitDegrees=1}
 {bone1Name=dyn_chain05, bone2Name=dyn_chain04, swing1LimitDegrees=45, swing2LimitDegrees=45, twistLimitDegrees=1}
 {bone1Name=dyn_chain04, bone2Name=dyn_chain03, swing1LimitDegrees=45, swing2LimitDegrees=45, twistLimitDegrees=1}
]
